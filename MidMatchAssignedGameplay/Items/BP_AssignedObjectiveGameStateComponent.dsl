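; Game state component that spawns mid-match assigned loot cache objectives for squads in Fortnite.
; When playlist data is ready, hotfix-gated listeners track storm and creative storm phase steps, starting objectives at a configured phase, optionally delayed.
; OnAssignedObjectiveReady spawns the cache and passes it squad players and midpoints for analytics, with optional debug lines.
; The storm-phase trigger is marked temporary until a shared objective spawn manager exists.
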(event EventBeginPlay
  ;[0.5 0.5 0.5] (TEMP) Bind to the Game Storm Phase in order to time out when the objective should be spawned - during the first shrink phase
  ; Should be defunct after there is a shared Midmatch Objective spawn manager
  ;;[1 1 1] EventBeginPlay
  ;; Check if the Assigned Objectives are enabled and bind to check the storm phases
  (|CallParentFunction)
  (Class|FortAsyncActionCurrentPlaylistReady|PlaylistDataReadyAsync
    (:OnReady
      (if (ScalableFloat|GetValueasBool (Variables|Hotfixes|Gethf_ObjectiveSpawningEnabled))
        ;;;[1 1 1] Bind to game phase step changed to listen for new storm circles
        (bind _returnvalue (GameplayEventRouter|GetEventRouter "/Script/FortniteGame.FortGameState" (Components|GetOwner)))
        (bind (_asynctaskproxy _payload) (GameplayEventRouter|StartListeningToStatefulEvent _returnvalue "/Script/FortniteGame.FortGamePhaseStepUpdatedEvent")
          (:then
            (Variables|Default|SetGamePhaseStepUpdatedAsyncTask _asynctaskproxy)
            (bind (_asynctaskproxy_2 _payload_2) (GameplayEventRouter|StartListeningToStatefulEvent _returnvalue "/Script/FortniteGame.FortCreativeStormShieldGamePhaseStepUpdatedEvent")
              (:then
                (Variables|Default|SetCreativeStormShieldGamePhaseStepUpdatedAsyncTask _asynctaskproxy_2))
              (:OnEventReceived
                _)
              (:OnSavedEventStateExists
                _))
            (CallFunction|OnGamePhaseStepChanged (.GamePhaseStep (Utilities|Struct|BreakFortCreativeStormShieldGamePhaseStepUpdatedEvent _payload_2))))
          (:OnEventReceived
            _)
          (:OnSavedEventStateExists
            _))
        (CallFunction|OnGamePhaseStepChanged (Utilities|Struct|BreakFortGamePhaseStepUpdatedEvent _payload))))))

(event Custom|OnGamePhaseStepChanged
  (params (GamePhaseStep :type EAthenaGamePhaseStep%20Enum))
  (bind _hf_delayfromstormphasetospawn (Variables|Default|GetHF_DelayFromStormPhaseToSpawn))
  (bind _returnvalue (ScalableFloat|GetValueatLevel _hf_delayfromstormphasetospawn))
  ;[0.5 0.5 0.5] (TEMP) Bind to the Game Storm Phase in order to time out when the objective should be spawned - during the first shrink phase
  ; Should be defunct after there is a shared Midmatch Objective spawn manager
  ;;[1 1 1] OnGamePhaseStepChanged
  ;; Starts the assigned objective based on the storm phase
  (if (ScalableFloat|GetValueasBoolwithContext (Variables|Default|GetHF_AllowedInMode))
    (if (== (Math|Conversions|ToInteger(Byte) GamePhaseStep) (ScalableFloat|GetValueasInteger (Variables|Hotfixes|GetHF_StormPhaseStepToSpawn)))
      (if (== (SafeZone|GetCurrentPhaseIndex (SafeZone|GetSafeZoneIndicator)) (ScalableFloat|GetValueasInteger (Variables|Hotfixes|GetHF_StormPhaseCountToSpawn)))
        (if (> _returnvalue 0)
          (Utilities|Time|SetTimerbyEvent (EventDispatchers|CreateEvent :!event Custom|DebugStartAssignedObjectives) _returnvalue)
          (else
            (CallFunction|DebugStartAssignedObjectives)))))))

(event EventEndPlay
  (params (EndPlayReason :type EEndPlayReason%20Enum :default "Destroyed"))
  ;[0.5 0.5 0.5] (TEMP) Bind to the Game Storm Phase in order to time out when the objective should be spawned - during the first shrink phase
  ; Should be defunct after there is a shared Midmatch Objective spawn manager
  ;;[1 1 1] EventEndPlay
  ;; Unbind events, clear timers, etc.
  (|CallParentFunction EndPlayReason)
  ;;;[1 1 1] unbind
  (bind _gamephasestepupdatedasynctask (Variables|Default|GetGamePhaseStepUpdatedAsyncTask)
    (:then
      (GameplayEventRouter|StopListeningtoStatefulEvent _gamephasestepupdatedasynctask))
    (:else
      _))
  (bind _creativestormshieldgamephasestepupdatedasynctask (Variables|Default|GetCreativeStormShieldGamePhaseStepUpdatedAsyncTask)
    (:then
      (GameplayEventRouter|StopListeningtoStatefulEvent _creativestormshieldgamephasestepupdatedasynctask))
    (:else)))

(event Custom|DebugDrawSquadAvgLocs
  ;[1 1 1] Debug/Cheats
  (bind _asfort_game_mode_athena (Utilities|Casting|CastToFortGameModeAthena (Game|GetGameMode)))
  (bind _squadids (Class|FortGameModeAthena|GetSquadIds _asfort_game_mode_athena))
  (for _array_element _squadids
    (bind _returnvalue (CallFunction|CalculateSquadAvgLocation _array_element))
    (Rendering|Debug|DrawDebugLine _returnvalue (+ _returnvalue "0.000000,0.000000,1000.000000") "(R=0.000000,G=0.379104,B=0.998264,A=1.000000)" 3e+01 1e+01)))

(event Custom|DebugStartAssignedObjectives
  ;[1 1 1] Debug/Cheats
  (CallFunction|StartAssignedObjectives))

(fn BuildDataRegistryResolverScope
  (bind _self self)
  (bind _returnvalue (Components|GetOwner _self))
  (bind _returnvalue_2 (Minigame|GetMinigame _returnvalue :!class /Script/FortniteGame.MinigameBlueprintLibrary))
  ;[1 0.8475 0.2776] Gives the name of the volume as a scope to the data registry. Needed for calls to GetValueWithContext to work
  (Utilities|IsValid _returnvalue_2
    (:Is%20Valid
      (return (Utilities|Array|MakeArray (SpatialGameplay|GetSpatialGameplayInterfaceFNameforActor _returnvalue_2)) 0 true))
    (:Is%20Not%20Valid
      (return 0 0 false))))

(fn GetCacheContextOverride
  ;[0.4163 1 0.9801] We don't need to return anything here since we don't need to override the Cache Context.
  (return 0))

(fn OnAssignedObjectiveReady
  (params (ObjectiveData :type Fort%20Assigned%20Objective%20Data%20Structure%20%28by%20ref%29))
  (bind (_assignedsquaddataarray _objectivelocation) (Utilities|Struct|BreakFortAssignedObjectiveData ObjectiveData))
  (Utilities|Array|Clear (LocalVariables|GetAssignedSquadPlayers))
  (Utilities|Array|Clear (LocalVariables|GetAssignedSquadMidpoints_Analytics))
  (Utilities|FlowControl|Sequence
    (:then_0
      ;[1 1 1] Debug Draw - Draw thick lines so that squad and objective locations can be seen easily from a distance (like player pings)
      (if (Variables|Default|GetbDebug)
        (LocalVariables|SetRandColor (Math|Color|MakeColor (Math|Random|RandomFloat) (Math|Random|RandomFloat) (Math|Random|RandomFloat)))
        (for _array_element _assignedsquaddataarray
          (Rendering|Debug|DrawDebugLine (.CachedSquadAvgLocation (Utilities|Struct|BreakFortAssignedObjectiveSquadData _array_element)) (+ (.CachedSquadAvgLocation (Utilities|Struct|BreakFortAssignedObjectiveSquadData _array_element)) "0.000000,0.000000,20000.000000") (LocalVariables|GetRandColor) 3e+02 3e+02))
        (Rendering|Debug|DrawDebugLine _objectivelocation (+ _objectivelocation "0.000000,0.000000,25000.000000") (LocalVariables|GetRandColor) 3e+02 5e+02)))
    (:then_1
      ;[0.3615 1 0.5352] Spawn the objective and let it know about its players
      (bind _returnvalue (Game|SpawnActorfromClass "/MidMatchAssignedGameplay/Items/Environmental/BGA_MidMatchAssignedObjective_Interaction_LootCache.BGA_MidMatchAssignedObjective_Interaction_LootCache_C" (Math|Conversions|ToTransform(Vector) _objectivelocation) "Undefined" "OverrideRootScale" 0 0 0))
      (Utilities|IsValid _returnvalue
        (:Is%20Valid
          (bind _asfort_game_mode_athena (Utilities|Casting|CastToFortGameModeAthena (Game|GetGameMode)))
          ;;[0.4596 1 0.8486] Convert squadIds to squad player states
          (for _array_element_2 _assignedsquaddataarray
            (bind _squadmembers (Class|FortGameModeAthena|GetSquadMembers _asfort_game_mode_athena (Utilities|Struct|BreakFortAssignedObjectiveSquadData _array_element_2)))
            (if (Utilities|Array|IsNotEmpty _squadmembers)
              (Utilities|Array|AppendArray (LocalVariables|GetAssignedSquadPlayers) _squadmembers)
              (Utilities|Array|Add (LocalVariables|GetAssignedSquadMidpoints_Analytics) (.CachedSquadAvgLocation (Utilities|Struct|BreakFortAssignedObjectiveSquadData _array_element_2))) ; used for analytics
            ))
          (Class|BGAMidMatchAssignedObjectiveInteractionParent|InitAssignedObjectiveonCreation _returnvalue (LocalVariables|GetAssignedSquadPlayers) (LocalVariables|GetAssignedSquadMidpoints_Analytics)))))
    (:then_2
      (return))))
